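# Niagara system for the Kepler emote that produces a light trickle of small, short-lived sprite confetti.
# The only enabled emitter, Confetti, spawns by probability-gated rate in a small sphere with random velocity biased toward +X, then applies drag and gravity and shrinks sprites late in life.
# ScaleColor multiplies only green by a random offset plus an age-driven ramp.
# LingerConfetti, RibbonConfetti and Sparkles emitters are present but disabled.
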NiagaraSystem /BRCosmetics/Effects/Fort_Effects/Emotes/Kepler/NS_Emote_Kepler_ConfettiTrickle
dependencies {
 usedDataInterfaces [&NiagaraDataInterfaceCurve, &NiagaraDataInterfaceColorCurve]
 usedDynamicInputs []
 usedModules [
  &SystemState
  &self:SystemSpawnScript.SystemScriptSource.SystemScriptGraph.NiagaraNodeAssignment_0.SetVariables_#1
 ]
}
emitters {
 Confetti {
  emitterUpdateScript [
   &EmitterState()
   &SpawnBurst_Instantaneous(Spawn Count=25, Use Spawn Probability=false) {enabled=false}
   &SpawnRate(Spawn Probability=0.25, SpawnRate=10, Use Spawn Probability=true)
  ]
  particleSpawnScript [
   /Niagara/Modules/Spawn/Initialization/InitializeParticle(
    Color=&RandomRangeLinearColor(Minimum=<rgba 0.5 0.305 0 1>, Maximum=<rgba 0.75 0.305 0 1>)
    Lifetime=&RandomRangeFloat(Minimum=0.2, Maximum=1)
    Module.Write Color=true
    Module.Write Lifetime=true
    Module.Write Mass=true
    Module.Write Material Random=true
    Module.Write Position=false
    Module.Write Scale=false
    Module.Write SpriteRotation=true
    Module.Write SpriteSize=true
    Position=&SimulationPosition()
    Sprite Rotation=&RandomRangeFloat(Minimum=0, Maximum=360)
    Sprite Size=&Vector2DFromFloat(Value=&RandomRangeFloat(Minimum=3, Maximum=6))
   )
   &SphereLocation(
    MaskSpawnGroup=false
    Override Seed=false
    Sphere Radius=&RandomRangeFloat(Minimum=0, Maximum=15)
    Surface Only=false
    UseNonUniformScale=false
    UseOffset=false
    UseSphereOrientation=false
   )
   &AddVelocityFromPoint(Module.Use Falloff Distance=false, Velocity Strength=&RandomRangeFloat(Minimum=10, Maximum=35)) {enabled=false}
   &AddVelocity(Velocity=&RandomRangeVector(Minimum=<-5 -50 -50>, Maximum=<75 50 50>))
   &AddVelocity(Velocity=$System.DirectionalVelocity) {enabled=false, module=AddVelocity001}
  ]
  particleUpdateScript [
   &ParticleState()
   &Drag(Drag=&RandomRangeFloat(Minimum=2, Maximum=4), Use Linear Drag=true, Use Rotational Drag=false)
   &GravityForce(Gravity=&RandomRangeVector(Minimum=<0 0 -150>, Maximum=<1 1 -50>))
   &SolveForcesAndVelocity(Clamp Velocity=false, Limit Acceleration=false)
   &ScaleSpriteSize(Scale Factor=&Vector2DFromFloat(Value=&FloatFromCurve(FloatCurve=DataInterface() {propertyValues {Curve {keys [
          {arriveTangent=-0.4596, interpMode=RCIM_Cubic, leaveTangent=-0.4596, tangentMode=RCTM_User, time=0.55, value=1}
          {arriveTangent=-4.9031, interpMode=RCIM_Cubic, leaveTangent=-4.9031, tangentMode=RCTM_User, time=1, value=0}
         ]}}})))
   &ScaleColor(
    Scale RGB=&MakeVector(
     X=1
     Y=&Add_Float(
      A=&RandomRangeFloat(Minimum=0, Maximum=1000)
      B=&Lerp_Float(A=0, B=&RandomRangeFloat(Minimum=25, Maximum=150), Alpha=$Particles.NormalizedAge)
     )
     Z=1
    )
    ScaleA=false
    ScaleRGB=true
   )
   /Niagara/Modules/Update/Color/ScaleColorBySpeed() {enabled=false}
  ]
  rendererClasses [&NiagaraSpriteRendererProperties]
  renderers [{rendererClass=&NiagaraSpriteRendererProperties}]
 }
 LingerConfetti {
  bEnabled=false
  emitterUpdateScript [
   &EmitterState()
   &SpawnBurst_Instantaneous(Spawn Count=25, Use Spawn Probability=false)
  ]
  particleSpawnScript [
   &InitializeParticle(
    Color=&RandomRangeLinearColor(Minimum=<rgba 0.5 0.305 0 1>, Maximum=<rgba 0.75 0.305 0 1>)
    Lifetime=&UniformRangedFloat(Minimum=1.5, Maximum=3)
    Module.Write Color=true
    Module.Write Lifetime=true
    Module.Write Mass=true
    Module.Write Material Random=true
    Module.Write Position=false
    Module.Write Scale=false
    Module.Write SpriteRotation=false
    Module.Write SpriteSize=true
    Position=&SimulationPosition()
    Sprite Size=&Vector2DFromFloat(Value=&UniformRangedFloat(Minimum=2, Maximum=4))
   )
   &SphereLocation(
    MaskSpawnGroup=false
    Override Seed=false
    Sphere Radius=2
    Surface Only=false
    UseNonUniformScale=false
    UseOffset=false
    UseSphereOrientation=false
   )
   &AddVelocityFromPoint(Module.Use Falloff Distance=false, Velocity Strength=&RandomRangeFloat(Minimum=50, Maximum=300))
   &AddVelocity(Velocity=&Multiply_VectorByFloat(Vector=$System.DirectionalVelocity, Float=&RandomRangeFloat(Minimum=0.75, Maximum=2)))
  ]
  particleUpdateScript [
   &ParticleState()
   &GravityForce(Gravity=<0 0 -200>)
   /Niagara/Modules/Update/Forces/CurlNoiseForce(Noise Frequency=0.02, Noise Strength=200, Pan Noise=false)
   &Drag(Drag=5, Use Linear Drag=true, Use Rotational Drag=false)
   &SolveForcesAndVelocity(Clamp Velocity=false, Limit Acceleration=false)
   &ScaleSpriteSize(Scale Factor=&Vector2DFromFloat(Value=&FloatFromCurve(FloatCurve=DataInterface() {propertyValues {Curve {keys [
          <at 0 0 RCIM_Cubic>
          {arriveTangent=-0.0623, interpMode=RCIM_Cubic, leaveTangent=-0.0623, tangentMode=RCTM_User, time=0.04, value=1}
          {arriveTangent=-0.0083, interpMode=RCIM_Cubic, leaveTangent=-0.0083, tangentMode=RCTM_User, time=0.5, value=1}
          <at 1 0 RCIM_Cubic>
         ]}}})))
   &ScaleColor(
    Scale RGB=&MakeVector(
     X=1
     Y=&Add_Float(
      A=&RandomRangeFloat(Minimum=0, Maximum=1000)
      B=&Lerp_Float(A=0, B=&RandomRangeFloat(Minimum=25, Maximum=400), Alpha=$Particles.NormalizedAge)
     )
     Z=1
    )
    ScaleA=false
    ScaleRGB=true
   )
  ]
  rendererClasses [&NiagaraSpriteRendererProperties]
  renderers [{rendererClass=&NiagaraSpriteRendererProperties}]
 }
 RibbonConfetti {
  bEnabled=false
  emitterUpdateScript [
   &EmitterState()
   &SpawnBurst_Instantaneous(Spawn Count=5, Use Spawn Probability=false) {enabled=false}
   &SpawnRate(Spawn Probability=0.5, SpawnRate=0.7, Use Spawn Probability=true)
  ]
  particleSpawnScript [
   &InitializeParticle(
    Color=&MakeCustomLinearColorFromBool(
     Bool=&RandomBool()
     True Color=&RandomRangeLinearColor(Minimum=<rgba 0 0.305 0 1>, Maximum=<rgba 0.4 0.305 0 1>)
     False Color=&RandomRangeLinearColor(Minimum=<rgba 0.85 0.305 0 1>, Maximum=<rgba 1 0.305 0 1>)
    )
    Lifetime=&RandomRangeFloat(Minimum=0.5, Maximum=1)
    Module.Write Color=true
    Module.Write Lifetime=true
    Module.Write Mass=true
    Module.Write Material Random=true
    Module.Write Position=false
    Module.Write Scale=false
    Module.Write SpriteRotation=true
    Module.Write SpriteSize=true
    Position=&SimulationPosition()
    Sprite Rotation=&RandomRangeFloat(Minimum=0, Maximum=360)
    Sprite Size=&Vector2DFromFloat(Value=&RandomRangeFloat(Minimum=5, Maximum=20))
   )
   &SphereLocation(
    MaskSpawnGroup=false
    Override Seed=false
    Sphere Radius=&RandomRangeFloat(Minimum=5, Maximum=15)
    Surface Only=false
    UseNonUniformScale=false
    UseOffset=false
    UseSphereOrientation=false
   )
   &AddVelocityFromPoint(Module.Use Falloff Distance=false, Velocity Strength=&RandomRangeFloat(Minimum=25, Maximum=100))
   &AddVelocity(Velocity=$System.DirectionalVelocity)
  ]
  particleUpdateScript [
   &ParticleState()
   &Drag(Drag=5, Use Linear Drag=true, Use Rotational Drag=false)
   &GravityForce(Gravity=&RandomRangeVector(Minimum=<0 0 -360>, Maximum=<1 1 -200>))
   &SolveForcesAndVelocity(Clamp Velocity=false, Limit Acceleration=false)
   &ScaleSpriteSize(Scale Factor=&Vector2DFromFloat(Value=&FloatFromCurve(FloatCurve=DataInterface() {propertyValues {Curve {keys [
          {arriveTangent=-0.4596, interpMode=RCIM_Cubic, leaveTangent=-0.4596, tangentMode=RCTM_User, time=0.55, value=1}
          {arriveTangent=-4.0966, interpMode=RCIM_Cubic, leaveTangent=-4.0966, tangentMode=RCTM_User, time=1, value=0}
         ]}}})))
  ]
  rendererClasses [&NiagaraSpriteRendererProperties]
  renderers [{rendererClass=&NiagaraSpriteRendererProperties}]
 }
 Sparkles {
  bEnabled=false
  emitterUpdateScript [
   &EmitterState()
   &SpawnBurst_Instantaneous(Spawn Count=50, Use Spawn Probability=false) {enabled=false}
   &SpawnRate(Spawn Probability=0.1, SpawnRate=8.6, Use Spawn Probability=true)
  ]
  particleSpawnScript [
   &InitializeParticle(
    Color=&HueShiftLinearColor(
     Hue=0
     Saturation=&RandomRangeFloat(Minimum=5, Maximum=10)
     Value=&RandomRangeFloat(Minimum=1, Maximum=10)
     Linear Color=&RandomRangeLinearColor(Minimum=<rgba 0 0 0 1>, Maximum=<rgba 1 1 1 1>, Color Channel Mode=::NewEnumerator2)
    )
    Lifetime=&Power(a=&RandomRangeFloat(Minimum=0.46, Maximum=1), b=2)
    Module.Write Color=true
    Module.Write Lifetime=true
    Module.Write Mass=true
    Module.Write Material Random=true
    Module.Write Position=false
    Module.Write Scale=false
    Module.Write SpriteRotation=true
    Module.Write SpriteSize=true
    Position=&SimulationPosition()
    Sprite Rotation=&RandomRangeFloat(Minimum=0, Maximum=360)
    Sprite Size=&Vector2DFromFloat(Value=&RandomRangeFloat(Minimum=1, Maximum=5))
   )
   &SphereLocation(
    MaskSpawnGroup=false
    Override Seed=false
    Sphere Radius=&RandomRangeFloat(Minimum=2, Maximum=5)
    Surface Only=false
    UseNonUniformScale=false
    UseOffset=false
    UseSphereOrientation=false
   )
   &AddVelocityFromPoint(Module.Use Falloff Distance=false, Velocity Strength=&RandomRangeFloat(Minimum=50, Maximum=150))
   &AddVelocity(Velocity=&Multiply_VectorByFloat(Vector=$System.DirectionalVelocity, Float=0.75))
  ]
  particleUpdateScript [
   &ParticleState()
   &Drag(Drag=5, Use Linear Drag=true, Use Rotational Drag=false)
   &GravityForce(Gravity=&RandomRangeVector(Minimum=<0 0 -200>, Maximum=<1 1 -500>))
   &SolveForcesAndVelocity(Clamp Velocity=false, Limit Acceleration=false)
   &ScaleColor(Scale Alpha=&Lerp_Float(A=1, B=0, Alpha=$Particles.NormalizedAge), ScaleA=true, ScaleRGB=false)
   &ScaleColor(
    Scale RGB=&MakeVectorFromLinearColorRGB(Color=&ColorFromCurve(ColorCurve=DataInterface() {propertyValues {
        AlphaCurve {keys [<at 0 0 RCIM_Cubic>,<at 0.052 1>,<at 0.1435 0.1>,<at 0.2605 1>,<at 0.3836 0.1>,<at 0.4918 1>,<at 0.6968 0.1677>,<at 1 0 RCIM_Cubic>]}
        BlueCurve {keys [<at 0 16.6566 RCIM_Cubic>,<at 0.361 3.2515>,<at 0.4146 0.6269>,<at 1 0 RCIM_Cubic>]}
        GreenCurve {keys [<at 0 0 RCIM_Cubic>,<at 0.361 1.4098>,<at 0.4146 0.2926>,<at 1 0.4788 RCIM_Cubic>]}
        RedCurve {keys [<at 0 20 RCIM_Cubic>,<at 0.361 5>,<at 0.4146 0.9906>,<at 1 1 RCIM_Cubic>]}
       }}))
    ScaleA=true
    ScaleRGB=true
   ) {enabled=false, module=ScaleColor001}
   &ScaleSpriteSize(Scale Factor=&Vector2DFromFloat(Value=&FloatFromCurve(FloatCurve=DataInterface() {propertyValues {Curve {keys [<at 0.0018 0.1957>,<at 0.0527 0.819>,<at 0.1169 0.6555>,<at 0.1748 0.9015>,<at 0.2394 0.582>,<at 0.332 1.0023>,<at 0.4947 0.3707>,<at 0.6727 1>,<at 1 0>]}}})))
  ]
  rendererClasses [&NiagaraSpriteRendererProperties]
  renderers [{rendererClass=&NiagaraSpriteRendererProperties}]
 }
}
summary {
 systemName=NS_Emote_Kepler_ConfettiTrickle
 emitters [
  {emitterName=Confetti, rendererClasses [&NiagaraSpriteRendererProperties]}
  {bEnabled=false, emitterName=RibbonConfetti, rendererClasses [&NiagaraSpriteRendererProperties]}
  {bEnabled=false, emitterName=Sparkles, rendererClasses [&NiagaraSpriteRendererProperties]}
  {bEnabled=false, emitterName=LingerConfetti, rendererClasses [&NiagaraSpriteRendererProperties]}
 ]
 userVariables [
  User.RibbonIDOffset {type=NiagaraInt32}
  User.SkeletalMesh {
   type=NiagaraDataInterfaceSkeletalMesh
   defaultValue {
    struct=NiagaraExt_VariableValue_DataInterface
    value {dataInterface=&self:NiagaraDataInterfaceSkeletalMesh_0}
   }
  }
 ]
}
